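# Idle Niagara effect for the RowLiaison backpack cosmetic: a single Popcorn sprite emitter that tosses kernels out of the pack in response to player activity.
# Per-frame spawn chances follow owner speed, crouching, weapon fire and landing.
# Kernels launch from five offsets on the skeletal mesh, colored by VariantPicker and tinted by TOD.
# A system timeline scales launch speed and gravity; kernels partly inherit owner movement, collide on selected quality levels, and shrink and die with distance.
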NiagaraSystem /BRCosmetics/Accessories/FORT_Backpacks/Backpack_RowLiaison/FX/NS_Backpack_RowLiaison_Idle_V2
dependencies {
 usedDataInterfaces [
  &NiagaraDataInterfaceArrayFloat3
  &NiagaraDataInterfaceCurve
  &NiagaraDataInterfacePlatformSet
  &NiagaraDataInterfaceSkeletalMesh
  &NiagaraDataInterfaceColorCurve
  &NiagaraDataInterfaceArrayFloat
 ]
 usedDynamicInputs [&ColorFromCurve]
 usedModules [
  &self:SystemSpawnScript.SystemScriptSource.SystemScriptGraph.NiagaraNodeAssignment_0.SetVariables_#1
  &SystemState
  &self:SystemSpawnScript.SystemScriptSource.SystemScriptGraph.NiagaraNodeAssignment_3.SetVariables_#2
  &self:SystemSpawnScript.SystemScriptSource.SystemScriptGraph.NiagaraNodeAssignment_2.SetVariables_#3
  &self:SystemSpawnScript.SystemScriptSource.SystemScriptGraph.NiagaraNodeAssignment_1.SetVariables_#4
  &Timeline
 ]
}
emitters {Popcorn {
  emitterUpdateScript [
   &EmitterState()
   &SpawnPerFrame(
    Spawn Count:NiagaraInt32=&MakeIntFromFloat(Float=&Multiply_Float(
      A=&Lerp_Float(
       A=0
       B=1
       Alpha=&ClampFloat(Float=&Divide_Float(A=&VectorLength(Vector=$Engine.Owner.Velocity), B=500))
      )
      B=$System.SpawnRateCurve
     ))
    Spawn Probability:NiagaraFloat=&MakeCustomFloatFromBool(
     Bool=&SetBoolByFloatComparison(A=2000, B=&VectorLength(Vector=$Engine.Owner.Velocity))
     True Float=0.2
     False Float=0
    )
    Use Spawn Probability=true
   ) {module=SpawnPerFrame003}
   &SpawnPerFrame(
    Spawn Probability:NiagaraFloat=&MakeCustomFloatFromBool(Bool=$User.IsCrouched, True Float=0.75, False Float=0)
    Use Spawn Probability=true
   ) {module=SpawnPerFrame001}
   &SpawnPerFrame(
    Spawn Probability:NiagaraFloat=&MakeCustomFloatFromBool(Bool=$User.IsWeaponFired, True Float=0.3, False Float=0)
    Use Spawn Probability=true
   ) {module=SpawnPerFrame002}
   &SpawnPerFrame(
    Spawn Probability:NiagaraFloat=&MakeCustomFloatFromBool(Bool=$User.bPawnLanded, True Float=1, False Float=0)
    Use Spawn Probability=true
   )
  ]
  particleSpawnScript [
   &PartitionParticles(Number Of Partitions=5)
   &InitializeParticle(
    Color=&MergeHelper_ColorSwitch(
     0=&LerpLinearColors(StartColor=<rgba 0.9473 0.7388 0.3503 1>, EndColor=<rgba 0.5711 0.2582 0.0513 1>, LerpFactor=0)
     1=&LerpLinearColors(StartColor=<rgba 0.6875 0.5403 0.2686 1>, EndColor=<rgba 0.5711 0.2582 0.0513 1>, LerpFactor=0)
     2=&LerpLinearColors(StartColor=<rgba 0.6875 0.6208 0.4977 1>, EndColor=<rgba 0.3281 0.3086 0.2957 1>, LerpFactor=0)
     3=?
     Index=&MakeIntFromFloat(Float=$User.VariantPicker, Conversion Mode=::NewEnumerator1)
     Num=::NewEnumerator1
    )
    Lifetime=&RandomRangeFloat(Minimum=1, Maximum=2)
    Sprite Size Mode=::NewEnumerator1
    Uniform Sprite Size Max=3
    Uniform Sprite Size Min=2.2
    UsePositionOffset=false
    Write Material Random=true
   )
   &SkeletalMeshLocation(
    MaskSpawnGroup=false
    OverrideRandomSeed=false
    Sampled Position Offset=&SelectVectorFromArray(
     Direct Array Index=$Particles.PartitionParticles.Partition
     Vector Selection Array=DataInterface() {propertyValues {FloatData [<0 -11 11>, <5 -16 8>, <-3 -15 9>, <8 -7 7>, <-2 -6 10>]}}
    )
    Skeletal Mesh=$System.Skeletal Mesh Reader
   )
   &AddVelocity(
    Cone Axis=<0 -0.7 1>
    Module.UseConeFalloff=false
    Module.UseVelDistribution=false
    Override Seed=false
    Velocity Mode=::NewEnumerator2
    Velocity Speed=&Multiply_Float(
     A=&Lerp_Float(
      A=120
      B=&RandomRangeFloat(Minimum=120, Maximum=180)
      Alpha=&MakeCustomFloatFromBool(Bool=$User.bPawnLanded, True Float=1, False Float=0)
     )
     B=&Lerp_Float(A=0.5, B=1, Alpha=$System.Timeline.Float)
    )
   )
  ]
  particleUpdateScript [
   &ParticleState()
   /SetVariables(
    Particles.MaxDistance:NiagaraFloat=&RandomRangeFloat(Minimum=125, Maximum=150)
    Particles.ParticleDistance:NiagaraFloat=&DistanceBetweenPositions(A=$Engine.Emitter.SimulationPosition, B=$Particles.Position)
   ) {isSetParameters=true}
   &ScaleSpriteSize(Scale Factor=&Vector2DFromFloat(Value=&FloatFromCurve(
      FloatCurve=DataInterface() {propertyValues {Curve {keys [<at 0.85 1 RCIM_Cubic>,<at 0.95 0>]}}}
      Scale Curve=&FloatFromCurve(
       FloatCurve=DataInterface() {propertyValues {Curve {keys [<at 0.75 1>,<at 1 0>]}}}
       CurveIndex=&ClampFloat(Float=&Divide_Float(A=$Particles.ParticleDistance, B=$Particles.MaxDistance))
      )
     )))
   &Collision(
    Advanced Aging Rate=0.5
    Age Colliding Particles=true
    Collision Enabled=&IsPlatformSetActive(New Platform Set=DataInterface() {propertyValues {Platforms {qualityLevelMask=-8}}})
    Enable Rest State=false
    Friction=0.35
    Module.EnableMaxCollisionCount=false
    Module.Randomize Collision Normal Vector=false
    Restitution=0.3
   )
   &GravityForce(Gravity=&Multiply_VectorByFloat(
     Vector=&RandomRangeVector(Minimum=<0 0 -600>, Maximum=<0 0 -400>)
     Float=&Lerp_Float(A=0.5, B=1, Alpha=$System.Timeline.Float)
    ))
   &InheritSourceMovement(
    Applied Force Scale=<0 0 0>
    Applied Position Scale=&VectorFromFloat(Value=&FloatFromCurve(FloatCurve=DataInterface() {propertyValues {Curve {
         defaultValue=1
         keys [{interpMode=RCIM_Cubic, tangentMode=RCTM_User, time=0, value=0.9}, {interpMode=RCIM_Cubic, tangentMode=RCTM_User, time=1, value=0.5}]
        }}}))
    ApplyForce=false
    ApplyPosition=true
    ApplyPositionFalloff=false
   ) {module=InheritSourceMovement001}
   &SolveForcesAndVelocity(Clamp Velocity=false, Limit Acceleration=false)
   &ScaleColor(
    Scale Alpha=&FloatFromCurve(FloatCurve=DataInterface() {propertyValues {Curve {keys [<at 0.7 1 RCIM_Cubic>,<at 1 0>]}}})
    Scale RGB=&Multiply_Vector(A=<1 1 1>, B=&MakeVectorFromLinearColorRGB(Color=$System.TOD))
    ScaleA=true
    ScaleRGB=true
   )
   &KillParticles(Kill Particles=&SetBoolByFloatComparison(A=$Particles.ParticleDistance, B=$Particles.MaxDistance))
  ]
  rendererClasses [&NiagaraSpriteRendererProperties]
  renderers [{rendererClass=&NiagaraSpriteRendererProperties}]
 }}
summary {
 systemName=NS_Backpack_RowLiaison_Idle_V2
 emitters [{emitterName=Popcorn, rendererClasses [&NiagaraSpriteRendererProperties]}]
 userVariables [
  User.bPawnLanded {type=NiagaraBool}
  User.bPawnSprintorSlide {type=NiagaraBool}
  User.IsCrouched {type=NiagaraBool}
  User.IsSkydiving {type=NiagaraBool}
  User.IsWeaponFired {type=NiagaraBool}
  User.LandBurst {type=NiagaraFloat, defaultValue=0, description=""}
  User.Switch {type=NiagaraFloat, defaultValue=1, description=""}
  User.VariantPicker {type=NiagaraFloat, defaultValue=0, description=""}
 ]
}
